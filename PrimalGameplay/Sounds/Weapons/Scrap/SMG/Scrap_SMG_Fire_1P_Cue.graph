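# First-person fire sound cue for the Scrap SMG, layering a close gunshot with a delayed shell casing impact.
# A mixer combines a random pick of four fire waves, with slight pitch and volume modulation, and a random pick of three shell impacts after a fixed delay.
# The impact waves are borrowed from Fortnite Machine Pistol sounds.
# It references first-person weapon attenuation, sound class and gunfire concurrency assets.
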SoundCue /PrimalGameplay/Sounds/Weapons/Scrap/SMG/Scrap_SMG_Fire_1P_Cue
root=Mixer_1
nodes [
 {id=Delay_0, children [Random_1], properties {delayMax=0.4, delayMin=0.4}}
 {id=Mixer_1, children [Modulator_0,Delay_0], properties {inputVolume [1,0.9]}}
 {id=Modulator_0, children [Random_0], properties {pitchMax=1.07, pitchMin=0.92, volumeMax=0.8, volumeMin=0.7}}
 {id=Random_0, children [WavePlayer_2,WavePlayer_3,WavePlayer_4,WavePlayer_8], properties {weights [1,1,1,1]}}
 {id=Random_1, children [WavePlayer_9,WavePlayer_5,WavePlayer_6], properties {weights [1,1,1]}}
 {id=WavePlayer_2, properties {soundWaveAssetPtr=&/Scrap_SMG_Fire_Close_02}}
 {id=WavePlayer_3, properties {soundWaveAssetPtr=&/Scrap_SMG_Fire_Close_03}}
 {id=WavePlayer_4, properties {soundWaveAssetPtr=&/Scrap_SMG_Fire_Close_04}}
 {
  id=WavePlayer_5
  properties {soundWaveAssetPtr=/Game/Sounds/Fort_Weapons/pistol/MachinePistol/SoundWaves/Machine_Pistol_EjectedShellImpacts_02}
 }
 {
  id=WavePlayer_6
  properties {soundWaveAssetPtr=/Game/Sounds/Fort_Weapons/pistol/MachinePistol/SoundWaves/Machine_Pistol_EjectedShellImpacts_03}
 }
 {id=WavePlayer_8, properties {soundWaveAssetPtr=&/Scrap_SMG_Fire_Close_05}}
 {
  id=WavePlayer_9
  properties {soundWaveAssetPtr=/Game/Sounds/Fort_Weapons/pistol/MachinePistol/SoundWaves/Machine_Pistol_EjectedShellImpacts_01}
 }
]
properties {
 attenuationSettings=&Fort_Weapon_Projectile_FirstPerson_Athena_Att
 bDebug=true
 maxDistance=2097152
 soundClassObject=&Fort_WeaponRanged_Fire_1P_PSM
 volumeMultiplier=0.8
 concurrencyOverrides {bLimitToOwner=true, maxCount=2, resolutionRule=StopOldest, volumeScale=0.8667}
 concurrencySet [&Weapon_Gunfire_FirstPerson_Conc]
}
