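# Sound cue for the Scrap SMG gunshot, layering close, medium and distant fire recordings by range.
# Each layer randomly picks one of four waves, varies it through a modulator, then applies its own distance attenuation.
# The close layer gets a third-person fire sound class, while medium and distant are mixed under a distant-fire class before the final mixer.
# Gunfire concurrency sets are assigned, alongside an owner-limited stop-oldest override.
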SoundCue /PrimalGameplay/Sounds/Weapons/Scrap/SMG/Scrap_SMG_Fire_Cue
root=Mixer_1
nodes [
 {
  id=Attenuation_118
  children [Modulator_3]
  properties {attenuationSettings=&Fort_Weapon_Projectile_Close_Athena_Att}
 }
 {
  id=Attenuation_119
  children [Modulator_2]
  properties {attenuationSettings=&Fort_Weapon_Projectile_Medium_Mono_Athena_Att}
 }
 {
  id=Attenuation_120
  children [Modulator_1]
  properties {attenuationSettings=&Fort_Weapon_Projectile_Distant_Mono_Athena_Att}
 }
 {id=Mixer_1, children [SoundClass_77, SoundClass_76], properties {inputVolume [1,0.9]}}
 {id=Mixer_2, children [Attenuation_119, Attenuation_120], properties {inputVolume [0.8,0.5]}}
 {id=Modulator_1, children [Random_5], properties {pitchMax=0.9, pitchMin=0.8, volumeMax=1.3, volumeMin=1.2}}
 {id=Modulator_2, children [Random_4], properties {pitchMax=1, pitchMin=0.9, volumeMax=2, volumeMin=2}}
 {id=Modulator_3, children [Random_0], properties {volumeMax=0.8, volumeMin=0.7}}
 {id=Random_0, children [WavePlayer_2,WavePlayer_3,WavePlayer_4,WavePlayer_0], properties {weights [1,1,1,1]}}
 {id=Random_4, children [WavePlayer_1, WavePlayer_6, WavePlayer_17, WavePlayer_18], properties {weights [1,1,1,1]}}
 {id=Random_5, children [WavePlayer_19, WavePlayer_20, WavePlayer_21, WavePlayer_22], properties {weights [1,1,1,1]}}
 {id=SoundClass_76, children [Mixer_2], properties {soundClassOverride=&Fort_WeaponRanged_Fire_Distant}}
 {
  id=SoundClass_77
  children [Attenuation_118]
  properties {soundClassOverride=&Fort_WeaponRanged_Fire_3P_PSM}
 }
 {id=WavePlayer_0, properties {soundWaveAssetPtr=&/Scrap_SMG_Fire_Close_05}}
 {id=WavePlayer_1, properties {soundWaveAssetPtr=&/Scrap_SMG_Fire_Medium_01}}
 {id=WavePlayer_17, properties {soundWaveAssetPtr=&/Scrap_SMG_Fire_Medium_03}}
 {id=WavePlayer_18, properties {soundWaveAssetPtr=&/Scrap_SMG_Fire_Medium_04}}
 {id=WavePlayer_19, properties {soundWaveAssetPtr=&/Scrap_SMG_Fire_Distant_01}}
 {id=WavePlayer_2, properties {soundWaveAssetPtr=&/Scrap_SMG_Fire_Close_02}}
 {id=WavePlayer_20, properties {soundWaveAssetPtr=&/Scrap_SMG_Fire_Distant_02}}
 {id=WavePlayer_21, properties {soundWaveAssetPtr=&/Scrap_SMG_Fire_Distant_03}}
 {id=WavePlayer_22, properties {soundWaveAssetPtr=&/Scrap_SMG_Fire_Distant_04}}
 {id=WavePlayer_3, properties {soundWaveAssetPtr=&/Scrap_SMG_Fire_Close_03}}
 {id=WavePlayer_4, properties {soundWaveAssetPtr=&/Scrap_SMG_Fire_Close_04}}
 {id=WavePlayer_6, properties {soundWaveAssetPtr=&/Scrap_SMG_Fire_Medium_02}}
]
properties {
 bDebug=true
 maxDistance=27500
 volumeMultiplier=0.8
 concurrencyOverrides {bLimitToOwner=true, maxCount=3, resolutionRule=StopOldest}
 concurrencySet [&Weapon_Gunfire_Local_Conc, &Weapon_Gunfire_Global_Conc]
}
