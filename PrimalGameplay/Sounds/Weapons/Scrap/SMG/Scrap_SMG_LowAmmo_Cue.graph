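# Sound cue for the low-ammo warning sound of the Scrap SMG weapon.
# A single wave player feeds one low-ammo sound wave straight to the output, with a slightly reduced volume.
# It routes through a Fortnite ranged-weapon low-ammo sound class and overrides concurrency so new instances are blocked rather than replacing those already playing.
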SoundCue /PrimalGameplay/Sounds/Weapons/Scrap/SMG/Scrap_SMG_LowAmmo_Cue
root=WavePlayer_0
nodes [{id=WavePlayer_0, properties {soundWaveAssetPtr=&/Scrap_SMG_LowAmmo_01}}]
properties {
 bOverrideConcurrency=true
 maxDistance=2097152
 soundClassObject=/Game/Sounds/FortSoundClassesAndModes/Fort_WeaponRanged_LowAmmo_Degraded_PSM
 volumeMultiplier=0.7
 concurrencyOverrides {resolutionRule=PreventNew}
}
