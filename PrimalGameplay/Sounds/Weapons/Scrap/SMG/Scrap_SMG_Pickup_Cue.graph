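# Sound cue for picking up the Scrap SMG weapon, classed as a loot-collect sound.
# A single pickup wave plays through a modulator that slightly randomizes pitch on each play.
# It uses weapon-pickup attenuation and reduced volume, and references a shared oldest-based pickup concurrency set alongside a single-instance, prevent-new override setting.
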SoundCue /PrimalGameplay/Sounds/Weapons/Scrap/SMG/Scrap_SMG_Pickup_Cue
root=Modulator_0
nodes [
 {id=Modulator_0, children [WavePlayer_4], properties {pitchMax=1.1, pitchMin=0.9}}
 {id=WavePlayer_4, properties {soundWaveAssetPtr=&/Scrap_SMG_Pickup_01}}
]
properties {
 attenuationSettings=&Fort_Weapon_Pickup_Att
 maxDistance=3500
 soundClassObject=&Fort_Loot_Collect
 volumeMultiplier=0.4615
 concurrencyOverrides {maxCount=1, resolutionRule=PreventNew}
 concurrencySet [&PickupMaximum_Oldest]
}
