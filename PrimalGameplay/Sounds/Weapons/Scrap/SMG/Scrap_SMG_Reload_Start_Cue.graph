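# Sound cue for the start of the Scrap SMG reload, routed through the Fort weapon foley sound class.
# A single wave player plays the Scrap_SMG_Reload_Start_01 wave through a modulator node at reduced volume.
# It references the shared weapon reload attenuation asset and overrides concurrency so that new instances are prevented when the limit is reached.
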SoundCue /PrimalGameplay/Sounds/Weapons/Scrap/SMG/Scrap_SMG_Reload_Start_Cue
root=Modulator_2
nodes [
 {id=Modulator_2, children [WavePlayer_0]}
 {id=WavePlayer_0, properties {soundWaveAssetPtr=&/Scrap_SMG_Reload_Start_01}}
]
properties {
 attenuationSettings=&Fort_Weapon_Reload_Att
 bOverrideConcurrency=true
 maxDistance=2250
 soundClassObject=&Fort_WeaponFoley
 volumeMultiplier=0.401
 attenuationOverrides {falloffDistance=4100, attenuationShapeExtents {x=900}}
 concurrencyOverrides {resolutionRule=PreventNew}
}
